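# Widget Blueprint for an Athena quest actor indicator that shows the audio log icon, apparently for audio log quest props.
# It derives from the native AthenaQuestActorIndicator and sets the inherited quest actor image to the audio log texture inside a small fixed-size box.
# It adds no logic: the event graph is empty, and GetIndicatedActor goes straight to a return with no actor connected.
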WidgetBlueprint /Game/Athena/Items/QuestParents/PropQuestActor/AthenaQuestAudioLogIndicator
graphs [
 EventGraph {links [], nodes []}
 GetIndicatedActor {nodes [
   GetIndicatedActor() {is=title K2Node_FunctionEntry}
   > Return Node!(execute, ReturnValue:Actor Object Reference)
  ]}
]
variables []
widgets {
 info {inheritedWidgetCount=2, parentClass=/Script/FortniteUI.AthenaQuestActorIndicator, rootWidgetClass=&SizeBox, widgetCount=3}
 widgets [
  SizeBox_0 {
   widgetClassPath=&SizeBox
   widget=SizeBox_0
   properties {bOverride_HeightOverride=true, bOverride_WidthOverride=true, heightOverride=32, widthOverride=32}
  }
  Image_QuestActorBrush {
   widgetClassPath=&Image
   bInherited=true
   bIsVariable=true
   parent=SizeBox_0
   widget=Image_QuestActorBrush
   properties {brush {
     bIsDynamicallyLoaded=false
     drawAs=Image
     imageType=NoImage
     mirroring=NoMirror
     resourceName=None
     resourceObject=/Game/Athena/HUD/Interaction/T_UI_NPC_AudioLog
     tiling=NoTile
     imageSize=<64 61>
     margin=<ltrb 0 0 0 0>
     outlineSettings {
      bUseBrushTransparency=false
      roundingType=HalfHeightRadius
      width=0
      color {colorUseRule=UseColor_Specified, specifiedColor=<rgba 0 0 0 0>}
      cornerRadii=<0 0 0 0>
     }
     tintColor {colorUseRule=UseColor_Specified, specifiedColor=<rgba 1 1 1 1>}
     uVRegion {bIsValid=false, max=<0 0>, min=<0 0>}
    }}
   slot {class=SizeBoxSlot}
  }
  TutorialHighlight {widgetClassPath=&TutorialHighlight}
 ]
}
