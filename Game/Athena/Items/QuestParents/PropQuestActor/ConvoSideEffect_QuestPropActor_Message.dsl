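; Conversation side effect that sends a configured message number to the prop quest actor taking part in a player's conversation.
; On the server, it validates the participant player controller, gets the NPC actor from the context, and casts it to BP_Athena_PropQuestActor_Parent to call ConversationMessage.
; Failed validity checks or casts write shipping log entries.
; The comments about daybreak components and trophy quests do not match this logic, which only uses the player controller as a guard.
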(fn ServerCauseSideEffect
  (params (Context :type Conversation%20Context%20Structure%20%28by%20ref%29))
  ;[0 1 0.9915] Server Cause Side Effect: Let the player's daybreak component know that the player has received information about the NPCQuestGiver's trophy quest through conversation.
  ;;[0 0.5202 1] Find the player controller that is participating in the conversation
  (bind _returnvalue (Class|FortConversationContextLibrary|GetParticipantPlayerController Context))
  (Utilities|IsValid _returnvalue
    (:Is%20Valid
      (bind _returnvalue_2 (Class|FortConversationContextLibrary|GetParticipantActorNPC (Variables|GetContext)))
      (bind _output_get (LocalVariables|SetParticipatingActorPawn _returnvalue_2))
      (Utilities|IsValid _output_get
        (:Is%20Valid
          (bind _asbp_athena_prop_quest_actor_parent (Utilities|Casting|CastToBP_Athena_PropQuestActor_Parent _output_get)
            (:then
              (Utilities|IsValid _asbp_athena_prop_quest_actor_parent
                (:Is%20Valid
                  (Conversation|ConversationMessage _asbp_athena_prop_quest_actor_parent (Variables|Default|GetMessage%20Number))
                  (return))
                (:Is%20Not%20Valid
                  ;;[1 0 0] Warning: Could not find NPCQuestGiverComponent
                  (Fortnite|FortShippingLog self "ServerCauseSideEffect: Could not get Quest Prop Actor" true))))
            (:CastFailed
              ;;[1 0 0] Warning: No controller found on Participating NPC actor
              (Fortnite|FortShippingLog self (Utilities|String|Append "ConvoSuccessSideEffect: not a prop quest parent actor:
" (Utilities|GetObjectName (LocalVariables|GetParticipatingActorPawn))) true))))
        (:Is%20Not%20Valid
          ;;[1 0 0] Warning: No participating NPC actor found
          (Fortnite|FortShippingLog self "ConvoSuccessSideEffect:  Could not get Participating Actor NPC" true))))
    (:Is%20Not%20Valid
      ;;[1 0 0] Warning: Could not find player controller for the player
      (Fortnite|FortShippingLog self "ServerCauseSideEffect: Could not get Player Controller for player " true))))
