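; Quest gameplay ability that applies a disguise to the player pawn by swapping in gender-specific cosmetic overrides.
; Activation toggles cosmetics on under authority, playing a Niagara swap effect and burst sound, and storing a looping audio component.
; Damage, DBNO, death, ability end or failed passive activation calls RemoveDisguise.
; That reverts cosmetics, stops the loop and removes the granting gameplay effect by tag; comments note faction changing was removed.
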(event Custom|DamagedWhileDisguised
  (params (DamagedActor :type Actor%20Object%20Reference) (Damage :type Float%20%28single-precision%29) (InstigatedBy :type Controller%20Object%20Reference) (DamageCauser :type Actor%20Object%20Reference) (HitLocation :type Vector) (FHitComponent :type Primitive%20Component%20Object%20Reference) (BoneName :type Name) (Momentum :type Vector))
  ;[1 1 1] Removed faction changing
  (CallFunction|RemoveDisguise))

(event Custom|DBNOWhileDisguised
  ;[1 1 1] Removed faction changing
  (CallFunction|RemoveDisguise))

(event Custom|DiedWhileDisguised
  (params (DamagedActor :type Actor%20Object%20Reference) (Damage :type Float%20%28single-precision%29) (InstigatedBy :type Controller%20Object%20Reference) (DamageCauser :type Actor%20Object%20Reference) (HitLocation :type Vector) (FHitComponent :type Primitive%20Component%20Object%20Reference) (BoneName :type Name) (Momentum :type Vector))
  ;[1 1 1] Removed faction changing
  (CallFunction|RemoveDisguise))

(event Ability|EventActivateAbility
  (CallFunction|ToggleCosmetics true))

(event Custom|ToggleCosmetics
  (params (ApplyingCosmetics :type Boolean))
  (bind _audiocomponent (Variables|Default|GetAudioComponent))
  (Ability|HasAuthority
    (:True
      (bind _returnvalue (GUID|GenerateGuidfromObject (Variables|Default|GetQuestID)))
      (bind _asfort_player_pawn_athena (Utilities|Casting|CastToFortPlayerPawnAthena (Ability|GetAvatarActorfromActorInfo self)))
      (Class|FortPlayerPawn|SwapPawnCosmeticItemandVariant _asfort_player_pawn_athena _returnvalue (select (.IsFemale (FortPlayerPawn|GetGender _asfort_player_pawn_athena)) (Variables|Default|GetCosmeticOverrides_Female) (Variables|Default|GetCosmeticOverrides_Male)) false (select ApplyingCosmetics "Active" "Inactive") false)
      (Niagara|SpawnSystemAttached "/Game/Effects/Fort_Effects/Effects/NS_CharacterSwap_FX.NS_CharacterSwap_FX" (Variables|/Script/Engine.Actor|GetRootComponent _asfort_player_pawn_athena) "None" "0, 0, 0" "0, 0, 0" "SnapToTarget" true)
      (Audio|SpawnSoundAttached (Variables|Default|GetBurstSwapSound) (Variables|/Script/Engine.Actor|GetRootComponent _asfort_player_pawn_athena))
      (if ApplyingCosmetics
        (bind _returnvalue_2 (Audio|SpawnSoundAttached (Variables|Default|GetLoopingSound) (Variables|/Script/Engine.Actor|GetRootComponent _asfort_player_pawn_athena)))
        (Variables|Default|SetAudioComponent _returnvalue_2)
        (else
          (Audio|Components|Audio|Stop _audiocomponent)
          (Components|DestroyComponent _audiocomponent))))))

(event Custom|RemoveDisguise
  (CallFunction|ToggleCosmetics false)
  (Ability|RemoveGameplayEffectFromOwnerWithGrantedTags (Variables|Default|GetGrantingGETags)))

(event Ability|EventOnEndAbility
  (params (bWasCancelled :type Boolean))
  (CallFunction|RemoveDisguise))

(event Ability|EventFailedToActivatePassiveAbility
  (params (ActorInfo :type Gameplay%20Ability%20Actor%20Info%20Structure))
  (CallFunction|RemoveDisguise))
