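; Gameplay ability that listens for player action events and reports them to the quest prop actor behind the effect that granted it.
; On event activation, it finds that BP_Athena_PropQuestActor_Parent as the effect causer in the granting effect context.
; It calls OnPlayerAction on that actor with the avatar pawn's player controller, data from the gameplay event and the pawn's gameplay tags.
; The graph contains no EndAbility call.
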(event Ability|EventActivateAbilityFromEvent
  (params (EventData :type Gameplay%20Event%20Data%20Structure%20%28by%20ref%29))
  (bind _asfort_player_pawn (Utilities|Casting|CastToFortPlayerPawn (Ability|GetAvatarActorfromActorInfo self)))
  (bind _asbp_athena_prop_quest_actor_parent (Utilities|Casting|CastToBP_Athena_PropQuestActor_Parent (Ability|EffectContext|GetEffectCauser (Ability|GetGrantedbyEffectContext))))
  (Class|BPAthenaPropQuestActorParent|OnPlayerAction _asbp_athena_prop_quest_actor_parent (Fortnite|GetFortPlayerControllerfromActor _asfort_player_pawn) (Utilities|Struct|BreakGameplayEventData EventData) (Variables|/Script/FortniteGame.FortPawn|GetGameplayTags _asfort_player_pawn)))
