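; Passive gameplay ability that reports a player's building actions to the prop quest actor behind the ability's granting effect.
; On activation it binds to the avatar controller's player-spawned-building event and tags each build as Athena.Quests.Build.NearQuestActor.
; A shared trigger casts the granting effect's causer to BP_Athena_PropQuestActor_Parent and calls OnPlayerAction with the controller, action tag and pawn gameplay tags.
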(event Ability|EventActivateAbility
  (Building|BindEventtoOnPlayerSpawnedBuildingActor (Fortnite|GetFortPlayerControllerfromActor (Ability|GetAvatarActorfromActorInfo self)) (delegate Custom|OnPlayerBuild)))

(event Custom|OnPlayerBuild
  (params (Spawner :type Fort%20Player%20Controller%20Object%20Reference) (SpawnedBuilding :type Building%20Actor%20Object%20Reference))
  (CallFunction|TriggeronPlayerAction Spawner "(TagName=\"Athena.Quests.Build.NearQuestActor\")"))

(event Custom|TriggerOnPlayerAction
  (params (PlayerController :type Fort%20Player%20Controller%20Object%20Reference) (ActionTag :type Gameplay%20Tag%20Structure))
  (bind _asbp_athena_prop_quest_actor_parent (Utilities|Casting|CastToBP_Athena_PropQuestActor_Parent (Ability|EffectContext|GetEffectCauser (Ability|GetGrantedbyEffectContext))))
  (Class|BPAthenaPropQuestActorParent|OnPlayerAction _asbp_athena_prop_quest_actor_parent PlayerController ActionTag (Variables|/Script/FortniteGame.FortPawn|GetGameplayTags (Game|Player|GetPlayerPawn PlayerController))))

(event Ability|EventFailedToActivatePassiveAbility
  (params (ActorInfo :type Gameplay%20Ability%20Actor%20Info%20Structure)))
